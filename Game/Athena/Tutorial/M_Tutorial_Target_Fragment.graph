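# Default Lit material for tutorial shooting target fragments, enabled for use on mesh particles.
# It samples the shared ShootingTargets prop diffuse and normal textures.
# A third sample reads the black-variant mask texture, and its RGB is inverted through a OneMinus node.
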Material /Game/Athena/Tutorial/M_Tutorial_Target_Fragment
graph {
 properties {bUsedWithMeshParticles=true, pixelDepthOffsetMode=PDOM_Legacy, usedShadingModels=Default Lit}
 nodes [
  OneMinus(None=TextureSample() {samplerType=SAMPLERTYPE_Masks, texture=/Game/Environments/Props/ShootingTargets/Textures/T_ShootingTargets_M_Black}.RGB)
  TextureSample() {texture=/Game/Environments/Props/ShootingTargets/Textures/T_ShootingTargets_D}
  TextureSample() {samplerType=SAMPLERTYPE_Normal, texture=/Game/Environments/Props/ShootingTargets/Textures/T_ShootingTargets_N}
 ]
}
